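# Translucent, unlit, two-sided spline material for glowing water tendrils in the Zero Point water river effect.
# World position offset combines SplineThicken, time-panned curl noise along normals and a camera offset.
# A vertex-color-offset panning plasma texture drives emissive, tinted between near and time-of-day-scaled far colors, and opacity, shaped by UV edge gradients.
# Named reroutes share distance fades and multipliers; further opacity branches use a collection-driven sphere mask, far fade and Fresnel.
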Material /Game/Environments/Hera/FX/ZeroPoint/WaterRiver/Materials/M_Env_ZeroPoint_WaterTendrils_Splines
graph {
 properties {bAutomaticallySetUsageInEditor=false, blendMode=BLEND_Translucent, shadingModel=MSM_Unlit, twoSided=true, usedShadingModels=Unlit}
 nodes [
  n1 = LinearInterpolate(A=ScalarParameter() {parameterName="WidthBase (S)"}, B=ScalarParameter() {parameterName=WidthBase_Far}, Alpha=NamedRerouteUsage())
  n2 = LinearInterpolate(A=ScalarParameter() {parameterName="WidthTip (S)"}, B=ScalarParameter() {parameterName=WidthTip_Far}, Alpha=NamedRerouteUsage())
  n3 = VertexNormalWS()
  n4 = VectorParameter() {parameterName=CurlNoise, defaultValue=<rgba 0.02 0.02 0.02 1>}
  n5 = VectorParameter() {parameterName=PanSpeed, defaultValue=<rgba 250 1500 -500 1>}
  n6 = Multiply(A=@n5.RGB, B=NamedRerouteUsage())
  n7 = NamedRerouteUsage()
  n8 = Add(
   A=Multiply(A=Time() {bIgnorePause=true, bOverride_Period=true, period=100}, B=LinearInterpolate(A=@n5.RGB, B=@n6, Alpha=@n7)) {constB=250}
   B=WorldPosition().XYZ
  )
  n9 = Multiply(
   A=VectorNoise(World Position=Multiply(A=@n4.RGB, B=@n8)) {noiseFunction=VNF_CurlALU}
   B=ScalarParameter() {defaultValue=1, parameterName=WPOCurlAmount}
  )
  Add(
   A=Add(
    A=MaterialFunctionCall(WidthBase=@n1, WidthTip=@n2) {materialFunction=&SplineThicken}.WorldPositionOffset
    B=Multiply(A=@n3, B=@n9)
   )
   B=MaterialFunctionCall(Offset Amount=ScalarParameter() {defaultValue=555, parameterName=Param_1}) {materialFunction=/Engine/Functions/Engine_MaterialFunctions02/WorldPositionOffset/CameraOffset}.Camera Offset
  )
  Constant() {r=1}
  n10 = ScalarParameter() {defaultValue=1, group=TimeOfDay, parameterName=TimeOfDay_Morning}
  n11 = ScalarParameter() {defaultValue=1, group=TimeOfDay, parameterName=TimeOfDay_Day}
  n12 = ScalarParameter() {defaultValue=1, group=TimeOfDay, parameterName=TimeOfDay_Evening}
  n13 = ScalarParameter() {defaultValue=1, group=TimeOfDay, parameterName=TimeOfDay_Night}
  n14 = LinearInterpolate(
   A=VectorParameter() {parameterName=Color, defaultValue=<rgba 1 0.3263 0 1>}.RGB
   B=Multiply(
    A=MaterialFunctionCall(Value_Morning=@n10, Value_Day=@n11, Value_Evening=@n12, Value_Night=@n13) {materialFunction=&Scalar_Time_of_Day}.Result
    B=VectorParameter() {parameterName=Color_Far, defaultValue=<rgba 1 0.0667 1 1>}.RGB
   )
   Alpha=NamedRerouteUsage()
  )
  n15 = TextureCoordinate()
  n16 = ComponentMask(None=@n15) {g=true}
  n17 = VertexColor()
  n18 = AppendVector(A=ComponentMask(None=@n15) {r=true}, B=Add(A=@n16, B=@n17.G))
  n19 = VectorParameter() {parameterName=PannerUVMult, defaultValue=<rgba 1 1 0 1>}
  n20 = AppendVector(A=ComponentMask(None=@n19.RGB) {r=true}, B=ComponentMask(None=@n19.RGB) {g=true})
  n21 = Constant()
  n22 = ScalarParameter() {defaultValue=1, parameterName=SpeedMultVertex}
  n23 = LinearInterpolate(A=@n22, B=Multiply(A=@n22, B=NamedRerouteUsage()), Alpha=NamedRerouteUsage())
  n24 = TextureSample(UVs=Panner(Coordinate=Multiply(A=@n18, B=@n20), Speed=AppendVector(A=@n21, B=@n23)) {speedY=1}) {samplerType=SAMPLERTYPE_LinearGrayscale, texture=/Game/Effects/Fort_Effects/Emotes/UltraLight/Texture/T_Plasma_A_B_Greyscale}
  MaterialFunctionCall(Emissive=Multiply(A=Multiply(A=@n14, B=@n24.RGB), B=ScalarParameter() {defaultValue=1, parameterName=Emissive})) {materialFunction=&MF_FX_EmissiveAdjustments}
  Multiply(A=@n16, B=@n17.G)
  Multiply(A=WorldPosition().XYZ, B=@n4.RGB)
  Multiply(B=NamedRerouteUsage())
  Multiply(B=NamedRerouteUsage())
  n25 = MaterialFunctionCall(UV Channel=TextureCoordinate()) {materialFunction=&LinearGradient}
  n26 = Multiply(A=@n25.UGradient, B=OneMinus(None=@n25.UGradient))
  n27 = Multiply(A=@n25.VGradient, B=OneMinus(None=@n25.VGradient))
  n28 = Multiply(A=Multiply(A=Multiply(A=@n26, B=@n27), B=@n24.G), B=ScalarParameter() {defaultValue=1, parameterName=FinalOpacity})
  n29 = Multiply(A=@n28, B=NamedRerouteUsage())
  n30 = TextureCoordinate() {coordinateIndex=1}
  n31 = Constant2Vector() {r=-0.1}
  Multiply(
   A=@n29
   B=SphereMask(
    A=Add(A=@n30, B=@n31)
    B=Constant2Vector() {g=0.5, r=0.5}
    Radius=CollectionParameter() {collection=&/MPC_ZP_WaterTendril_Opacity, parameterName=SphereMask_Radius}
   ) {attenuationRadius=0.5, hardnessPercent=0}
  )
  n32 = ScalarParameter() {defaultValue=2500, parameterName=FarFadeout_Distance}
  n33 = ScalarParameter() {defaultValue=-2500, parameterName=FarFadeout_Distance_Offset}
  Multiply(
   A=OneMinus(None=MaterialFunctionCall!(Blend Range=@n32, Start Offset=@n33) {materialFunction=/Engine/Functions/Engine_MaterialFunctions03/Particles/Distance_Blend}.Result)
   B=@n29
  )
  n34 = OneMinus(None=SphereMask(A=PixelDepth().R, B=Constant(), Radius=ScalarParameter() {defaultValue=1500, parameterName=FadeDistance}) {hardnessPercent=0})
  n35 = MaterialFunctionCall(Distance=ScalarParameter() {parameterName=Near_Camera_Fade_Distance}) {materialFunction=&Near_Camera_Fade}
  NamedRerouteDeclaration(None=Multiply(A=Clamp(None=@n34) {minDefault=0.05}, B=Multiply(A=@n35.Result, B=@n35.Result))) {name=FX_DistanceFade, nodeColor=<rgba 1 0 0.3047 1>}
  NamedRerouteDeclaration(None=ScalarParameter() {defaultValue=1, parameterName=DistanceValueMultiplier}) {name=DistanceValueMultiplier, nodeColor=<rgba 1 0.6797 0 1>}
  NamedRerouteDeclaration(None=MaterialFunctionCall(
    Fade_Start_Distance=ScalarParameter() {defaultValue=90000, parameterName=DistanceFade_End}
    Fade_End_Distance=ScalarParameter() {defaultValue=40000, parameterName=DistanceFade_Start}
   ) {materialFunction=&MF_Distance_SyncedFade_FX}.Result) {name=FadeBetweenDistances, nodeColor=<rgba 1 0 0.0234 1>}
  ScalarParameter() {defaultValue=1, parameterName=SphereMask_Radius}
  n36 = ScalarParameter() {defaultValue=1.5, parameterName="Exponent (S)"}
  SmoothStep(
   Min=ScalarParameter() {defaultValue=-0.144, parameterName=Param_2}
   Max=ScalarParameter() {defaultValue=1.508, parameterName=Param_3}
   Value=Multiply(
    A=@n28
    B=MaterialFunctionCall(Exponent=@n36) {materialFunction=&AxisAlignedFresnel}.Fresnel
   )
  )
 ]
}
